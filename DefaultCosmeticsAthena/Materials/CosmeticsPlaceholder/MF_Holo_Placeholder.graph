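# Material function that restyles incoming material attributes as a hologram for placeholder cosmetics, blended over the original by HologramColorBlend.
# Scanlines with wiggle-driven blocks, Fresnel, inner/outer glow colors, Z gradient and near-camera fade build color and opacity.
# Two scrolling vertical waves, optionally in stabilized screen space, add emissive glow.
# Static switches enable inner lighting, morph-target normals and vertex-color hiding through WPO, while Skin FX textures can drive the opacity mask.
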MaterialFunction /DefaultCosmeticsAthena/Materials/CosmeticsPlaceholder/MF_Holo_Placeholder
graph {
 properties {previewBlendMode=BLEND_Masked}
 nodes [
  n1 = NamedRerouteUsage()
  n2 = ScalarParameter() {
   defaultValue=0.9
   desc="Setting this too low will produce \"anotomically correct\" highlights - if you get what I'm saying"
   parameterName=RoughnessAmount
  }
  n3 = SetMaterialAttributes!(
   MaterialAttributes=@n1
   BaseColor=NamedRerouteUsage()
   Roughness=@n2
   EmissiveColor=NamedRerouteUsage()
   Opacity=Constant()
   OpacityMask=NamedRerouteUsage()
   Normal=NamedRerouteUsage()
   WorldPositionOffset=NamedRerouteUsage()
   CustomizedUV2=NamedRerouteUsage()
   CustomizedUV3=NamedRerouteUsage()
  ) {attributeSetTypes [
    #1
    #2
    #3
    #4
    #5
    #6
    #7
    #8
    #9
   ]}
  FunctionOutput(None=BlendMaterialAttributes(A=@n1, B=@n3, Alpha=ScalarParameter() {defaultValue=1, parameterName=HologramColorBlend}))
  n4 = NamedRerouteUsage()
  n5 = ScalarParameter() {defaultValue=0.3, parameterName=MainScanlinesOpacity}
  Multiply(A=@n4, B=@n5)
  n6 = ComponentMask(None=WorldPosition().XYZ) {b=true}
  n7 = ScalarParameter() {defaultValue=376, group=ZGradient, parameterName=StructureHeight}
  n8 = Add(
   A=ScalarParameter() {group=ZGradient, parameterName=ObjectPositionZ}
   B=Multiply(
    A=DynamicParameter() {defaultValue=<rgba 1 0 0 0>, paramNames [EdgeSpeed, BiasScale, Gradient Position, TriggerWiggle]}.Gradient Position
    B=@n7
   )
  )
  n9 = ScalarParameter() {defaultValue=0.5, group=ZGradient, parameterName=Gradient Thickness}
  n10 = Clamp(None=Divide(A=Subtract(A=@n6, B=@n8), B=Multiply(A=@n7, B=@n9)) {constB=384})
  n11 = Multiply(A=@n10, B=OneMinus(None=@n10))
  n12 = ScalarParameter() {defaultValue=4, group=ZGradient, parameterName=Gradient Brightness}
  NamedRerouteDeclaration(None=StaticSwitchParameter(True=Multiply(A=@n11, B=@n12) {constB=4}, False=Constant() {r=1}) {parameterName=Use Z Gradient?}) {name=ZGradient, nodeColor=<rgba 0 0.9062 1 1>}
  n13 = MaterialFunctionCall!(
   WPO Intensity A=VectorParameter() {parameterName=wiggle1, defaultValue=<rgba 0.2 0.4 0.3524 1>}.RGB
   WPO Intensity B=VectorParameter() {parameterName=wiggle1_1, defaultValue=<rgba 0 1 0 1>}.RGB
   Time
   Bias=StaticSwitchParameter(
    True=DynamicParameter() {defaultValue=<rgba 1 0 0 0>, paramNames [EdgeSpeed, BiasScale, Gradient Position, TriggerWiggle]}.TriggerWiggle
    False=Constant()
   ) {defaultValue=true, parameterName=Use Param4 for Wiggle}
  ) {materialFunction=/Game/Effects/Fort_Effects/Effects/Atlas/Materials/MF_Atlas_HologramWiggle_01}
  n14 = ScalarParameter() {defaultValue=0.65, parameterName=BlocksIntensity}
  n15 = MaterialFunctionCall!(
   Number of Scanlines=ScalarParameter() {defaultValue=150, parameterName=Scanline Amount}
   Scanline Speed
   "Scanlines Thickness (Bias)"=ScalarParameter() {defaultValue=-0.15, parameterName=ScanlineSize}
   Scanlines Dark Value=ScalarParameter() {defaultValue=0.5, parameterName=Scanlines Dark}
   Scanlines Light Value=ScalarParameter() {defaultValue=1, parameterName=Scanlines Light}
   Block Opacity=StaticSwitchParameter(True=Multiply(A=@n13.Intense 0-1, B=@n14), False=Constant()) {defaultValue=true, parameterName=Use Blocks?}
   UVs
   Time=Multiply(A=Time(), B=ScalarParameter() {defaultValue=0.1, parameterName=ScanlinesTimeMultiplier})
   Block Repeat
  ) {materialFunction=/Game/Effects/Fort_Effects/Effects/Atlas/Materials/MF_Atlas_Scanlines_01}
  NamedRerouteDeclaration(None=@n15.Result) {name=MainScanlines, nodeColor=<rgba 1 0.7734 0 1>}
  n16 = VectorParameter() {parameterName=Glow Color Inner, defaultValue=<rgba 0.4199 0.4465 0.7396 1>}
  n17 = Desaturation(None=NamedRerouteUsage(), Fraction=ScalarParameter() {defaultValue=0.3, parameterName=InnerBaseColorDesaturation})
  n18 = VectorParameter() {parameterName=Glow Color Outer, defaultValue=<rgba 3.1771 3.7137 5 1>}
  n19 = LinearInterpolate(
   A=@n18.RGB
   B=Multiply(A=@n18.RGB, B=NamedRerouteUsage())
   Alpha=ScalarParameter() {defaultValue=0.5, parameterName=OutterGlowBaseColorBias, sliderMax=1}
  )
  n20 = ScalarParameter() {parameterName=OutterBaseColorDesaturation}
  n21 = WorldPosition()
  n22 = MaterialFunctionCall() {materialFunction=&ObjectLocalBounds}
  n23 = Divide(
   A=Add(A=ComponentMask(None=@n21.XYZ) {b=true}, B=ScalarParameter() {parameterName=FresnelEdgesZ_AttenOffset}) {constB=10}
   B=Multiply(A=ComponentMask(None=@n22.Local Bounds Size) {b=true}, B=ScalarParameter() {parameterName=FresnelEdgesZ_AttenScale}) {constB=1.2}
  ) {constB=384}
  n24 = Add(
   A=Multiply(A=Saturate(None=@n23), B=NamedRerouteUsage()) {constA=1}
   B=Multiply(A=ScalarParameter() {defaultValue=0.5, parameterName=Secondary Line Brightness}, B=NamedRerouteUsage())
  )
  n25 = Multiply(A=NamedRerouteUsage(), B=LinearInterpolate(A=Add(A=@n16.RGB, B=@n17), B=Desaturation(None=@n19, Fraction=@n20), Alpha=Saturate(None=@n24)))
  n26 = Constant() {r=1}
  n27 = Saturate(None=NamedRerouteUsage())
  NamedRerouteDeclaration(None=Multiply(A=Multiply(A=@n25, B=@n26), B=LinearInterpolate(Alpha=@n27) {constA=0.1})) {name=OutBaseColor, nodeColor=<rgba 0 0.1094 1 1>}
  n28 = NamedRerouteUsage()
  n29 = NamedRerouteUsage()
  n30 = ScalarParameter() {defaultValue=0.8, parameterName=Opacity Bias Scale}
  n31 = Multiply(A=Add(A=Multiply(A=@n29, B=@n30), B=OneMinus(None=@n30)), B=NamedRerouteUsage())
  n32 = MaterialFunctionCall(Distance=ScalarParameter() {defaultValue=200, parameterName=Near Camera Fade Distance}) {materialFunction=&Near_Camera_Fade}
  n33 = Multiply(A=Constant() {r=1}, B=Multiply(A=Multiply(A=@n31, B=@n32.Result), B=Max(A=@n4, B=@n5)))
  n34 = NamedRerouteUsage()
  n35 = ScalarParameter() {defaultValue=1, parameterName=Secondary Line Opacity}
  NamedRerouteDeclaration(None=Add(A=Multiply(A=@n28, B=@n33), B=Multiply(A=@n34, B=@n35))) {name=OutOpacity, nodeColor=<rgba 0.5234 0 1 1>}
  n36 = NamedRerouteUsage()
  n37 = ScalarParameter() {defaultValue=0.2, parameterName=PrimaryWaveScale}
  n38 = Multiply(A=ScalarParameter() {defaultValue=1.5, parameterName=PrimaryWaveSpeed}, B=@n37)
  n39 = ComponentMask(None=MaterialFunctionCall!(ObjectPivotLocation, ObjectPositionOffset) {materialFunction=/Game/Characters/Player/Common/MaterialFunctions/Superlevel/ImprovedStabilizedScreenSpaceUVs}.RawUVs) {g=true}
  n40 = ComponentMask(None=Subtract(A=WorldPosition().XYZ, B=ActorPositionWS())) {b=true}
  n41 = Multiply(
   A=Multiply(A=StaticSwitchParameter(True=@n39, False=@n40) {defaultValue=true, parameterName=UseScreenSpaceVerticalWipe}) {constB=-0.01}
   B=@n37
  ) {constB=0.5}
  n42 = Frac(None=Add(A=Multiply(A=@n36, B=@n38) {constB=0.06}, B=@n41))
  n43 = SmoothStep(Min=ScalarParameter() {defaultValue=0.8, parameterName=PrimaryWaveBottomMin}, Value=OneMinus(None=@n42)) {constMin=0.5}
  n44 = SmoothStep(Min=ScalarParameter() {defaultValue=0.99, parameterName=PrimaryWaveTopMin}, Value=@n42) {constMin=0.95}
  NamedRerouteDeclaration(None=Max(A=Multiply(A=@n43, B=@n43), B=Multiply(A=@n44, B=@n44))) {name=VerticalWave, nodeColor=<rgba 0 1 0.2031 1>}
  n45 = ScalarParameter() {defaultValue=1.5, parameterName=SecondaryWaveScale}
  n46 = Multiply(A=@n41, B=@n45)
  n47 = Multiply(A=NamedRerouteUsage(), B=Multiply(A=ScalarParameter() {defaultValue=0.5, parameterName=SecondaryWaveSpeed}, B=@n45)) {constB=0.06}
  n48 = Sine(None=Subtract(A=Add(A=@n46, B=@n47)) {constB=0.075})
  NamedRerouteDeclaration(None=SmoothStep(Min=ScalarParameter() {parameterName=SecondaryWaveMin}, Value=Saturate(None=@n48)) {constMin=0.5}) {name=SecondaryWave, nodeColor=<rgba 0.125 0 1 1>}
  NamedRerouteDeclaration(None=StaticSwitchParameter(True=Constant(), False=NamedRerouteUsage()) {parameterName=Use Morph Targets?}) {name=BaseColor, nodeColor=<rgba 1 0 0.6797 1>}
  n49 = GetMaterialAttributes(None=NamedRerouteDeclaration(None=FunctionInput() {bUsePreviewValueAsDefault=true, inputName=InMaterial, inputType=FunctionInput_MaterialAttributes}) {name=InMaterial, nodeColor=<rgba 0 1 1 1>}) {attributeGetTypes [#1, #6, #3]}
  NamedRerouteDeclaration(None=@n49.EmissiveColor) {name=InEmissive, nodeColor=<rgba 1 0.3516 0 1>}
  NamedRerouteDeclaration(None=@n49.BaseColor) {name=InBaseColor, nodeColor=<rgba 0.3594 0 1 1>}
  NamedRerouteDeclaration(None=@n49.Normal) {name=InNormal, nodeColor=<rgba 0 1 0.0391 1>}
  n50 = NamedRerouteUsage()
  n51 = PixelNormalWS()
  NamedRerouteDeclaration(None=MaterialFunctionCall(
    Brightness=ScalarParameter() {defaultValue=1, parameterName=Fresnel Brightness, sliderMin=4}
    Exponent=ScalarParameter() {defaultValue=1.5, parameterName=Fresnel Exponent}
    Vertex / Pixel Normal=StaticSwitchParameter(True=@n50, False=@n51) {parameterName=Use Morph Targets?}
   ) {materialFunction=&MF_Fresnel_FX}.Fresnel Inverted) {name=Fresnel, nodeColor=<rgba 1 0 0.2812 1>}
  NamedRerouteDeclaration(None=@n15.Secondary Lines) {name=SecondaryScanlines, nodeColor=<rgba 0.4531 0 1 1>}
  n52 = ScalarParameter() {defaultValue=0.2, parameterName=Inner Opacity Dark}
  n53 = ScalarParameter() {defaultValue=1, parameterName=Inner Opacity Bright}
  n54 = Subtract(A=VectorParameter() {parameterName=ObjectPosition, defaultValue=<rgba 0 0 250 0>}.RGB, B=WorldPosition().XYZ)
  n55 = ConstantBiasScale(None=DotProduct(A=PixelNormalWS(), B=Normalize(VectorInput=@n54)))
  n56 = Constant() {r=1}
  NamedRerouteDeclaration(None=StaticSwitchParameter(True=LinearInterpolate(A=@n52, B=@n53, Alpha=@n55), False=@n56) {parameterName=Use Inner Light?}) {name=InnerOpacity, nodeColor=<rgba 0 0.6016 1 1>}
  n57 = ScalarParameter() {defaultValue=0.05, parameterName=Inner Lighting Dark}
  n58 = ScalarParameter() {defaultValue=3, parameterName=Inner Lighting Bright}
  NamedRerouteDeclaration(None=StaticSwitchParameter(True=LinearInterpolate(A=@n57, B=@n58, Alpha=@n55), False=@n56) {parameterName=Use Inner Light?}) {name=InnerLighting, nodeColor=<rgba 0.9219 1 0 1>}
  NamedRerouteDeclaration(None=Time()) {name=Time, nodeColor=<rgba 1 0.3281 0 1>}
  n59 = TextureObjectParameter() {
   parameterName=Morph Target Normals
   samplerType=SAMPLERTYPE_LinearColor
   texture=/Game/Global/Textures/Default/Blanks/T_BlackVectorDisplacement
  }
  n60 = TextureObjectParameter() {
   parameterName=Morph Target Offsets
   samplerType=SAMPLERTYPE_LinearColor
   texture=&T_BlackVectorDisplacement
  }
  n61 = MaterialFunctionCall!(
   Morph Animation=Multiply(A=ScalarParameter() {defaultValue=0.35, parameterName=Time Speed}, B=Time() {period=1})
   Morph Normal=@n59
   Morph Texture=@n60
   Number of Morph Targets=ComponentMask(None=TextureProperty(Tex=@n59)) {g=true}
   0-1 Animation Value?=StaticBool() {value=true}
   Morph Target UV
  ) {materialFunction=/Engine/Functions/Engine_MaterialFunctions02/MAXScripts/MS_VertexAnimationTools_MorphTargets}
  n62 = Constant()
  NamedRerouteDeclaration(None=StaticSwitchParameter(True=@n61."Custom UV 2 (Needed)", False=@n62) {parameterName=Use Morph Targets?}) {name=CustomUV2, nodeColor=<rgba 0 1 0.1328 1>}
  NamedRerouteDeclaration(None=StaticSwitchParameter(True=@n61."Custom UV 3 (Needed)", False=@n62) {parameterName=Use Morph Targets?}) {name=CustomUV3, nodeColor=<rgba 1 0 0.1875 1>}
  NamedRerouteDeclaration(None=StaticSwitchParameter(
    True=MaterialFunctionCall!(WPO) {materialFunction=/Game/Packages/Fortress_SharedMaterials/Base_Material_Functions/Fortnite_HideElementsWithVertexColor}.Result
    False=Constant()
   ) {parameterName=UseVertexMaskin}) {name=OutWPO, nodeColor=<rgba 0 0.5078 1 1>}
  n63 = NamedRerouteUsage()
  NamedRerouteDeclaration(None=StaticSwitchParameter(
    True=NamedRerouteUsage()
    False=MaterialFunctionCall!(TextureNormals=@n63) {materialFunction=/Game/UI/Frontend/Lobby/HeroSquad/MF_Kicks_UDIM_Normal_Override}.Result
   ) {parameterName=Use Morph Targets?}) {name=OutNormal, nodeColor=<rgba 0.8281 1 0 1>}
  n64 = TextureSampleParameter2D() {
   group=Skin FX
   parameterName=SkinFX_Mask
   samplerType=SAMPLERTYPE_Masks
   sortPriority=1
   texture=&T_WhiteMask
  }
  n65 = ScalarParameter() {group=HologramMasking, parameterName=UseSkinFXMaskForOpacity}
  n66 = TextureSampleParameter2D() {
   group=Skin FX
   parameterName=TechArtMask
   samplerType=SAMPLERTYPE_Masks
   sortPriority=1
   texture=&T_WhiteMask
  }
  NamedRerouteDeclaration(None=LinearInterpolate(
    A=LinearInterpolate(B=@n64.R, Alpha=@n65) {constA=1}
    B=@n66.R
    Alpha=ScalarParameter() {group=HologramMasking, parameterName=UseTechArtMaskForOpacity}
   ) {constA=1}) {name=OutOpacityMask, nodeColor=<rgba 0.8516 0 1 1>}
  NamedRerouteDeclaration(None=@n61."Pixel Shader World Vertex Normal (See tooltip)") {name=MorphTargetNormals, nodeColor=<rgba 0 1 0.5313 1>}
  n67 = VectorParameter() {parameterName=VerticalWaveColor, defaultValue=<rgba 0.1986 0.98 3 1>}
  n68 = NamedRerouteUsage()
  n69 = ScalarParameter() {defaultValue=1, parameterName=PrimaryWaveIntensity}
  n70 = NamedRerouteUsage()
  n71 = ScalarParameter() {defaultValue=0.3, parameterName=SecondaryWaveIntensity}
  n72 = Multiply(A=Add(A=Multiply(A=@n68, B=@n69) {constB=0.1}, B=Multiply(A=@n70, B=@n71) {constB=0.1}), B=NamedRerouteUsage())
  n73 = NamedRerouteUsage()
  n74 = Multiply(A=NamedRerouteUsage(), B=LinearInterpolate(Alpha=NamedRerouteUsage()) {constA=1, constB=1.1})
  NamedRerouteDeclaration(None=Add(A=Multiply(A=@n67.RGB, B=@n72), B=Add(A=@n73, B=@n74))) {name=OutEmissive, nodeColor=<rgba 1 0.2109 0 1>}
  NamedRerouteUsage()
  ObjectPositionWS()
 ]
}
