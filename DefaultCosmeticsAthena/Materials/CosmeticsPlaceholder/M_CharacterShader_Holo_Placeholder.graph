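# Placeholder holographic character material for Athena cosmetics, built on the 2023 character parent functions for skinned, morph, cloth and deformer meshes.
# An IsEye static switch picks the eye function or the PreFX parent output.
# A default-on SwizzleRoughnessMetallic switch re-routes roughness and metallic; its off branch takes raw PreFX, bypassing the eye choice.
# The result runs through MF_Holo_Placeholder, then the PostFX parent, which also receives auxiliary PreFX outputs.
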Material /DefaultCosmeticsAthena/Materials/CosmeticsPlaceholder/M_CharacterShader_Holo_Placeholder
graph {
 properties {
  bUseMaterialAttributes=true
  bUsedWithClothing=true
  bUsedWithMeshDeformer=true
  bUsedWithMorphTargets=true
  bUsedWithSkeletalMesh=true
  bUsedWithStaticLighting=true
  numCustomizedUVs=3
  physMaterial=/Game/FortressPhysicalMaterials/Player_Flesh
  pixelDepthOffsetMode=PDOM_Legacy
  subsurfaceProfile=/Game/Packages/Fortress_SharedMaterials/SubSurfaceProfile/Character_SubsurfaceProfile
  usedShadingModels=Default Lit
 }
 nodes [
  n1 = MaterialFunctionCall() {materialFunction=&MF_CharacterParent_2023_PreFX}
  n2 = GetMaterialAttributes(None=StaticSwitchParameter(
    True=MaterialFunctionCall() {materialFunction=/Game/Global/Materials/Character_Parent_2023/Eye_Parent_2023/MF_eyes_2023}.Result
    False=@n1.Result
   ) {parameterName=IsEye}) {attributeGetTypes [#1, #2]}
  n3 = SetMaterialAttributes!(MaterialAttributes=@n2.MaterialAttributes, Roughness=@n2.MaterialAttributes, Metallic=@n2.MaterialAttributes) {attributeSetTypes [#1, #2]}
  MaterialFunctionCall(
   Material=MaterialFunctionCall!(InMaterial=StaticSwitchParameter(True=@n3, False=@n1.Result) {defaultValue=true, parameterName=SwizzleRoughnessMetallic}) {materialFunction=&/MF_Holo_Placeholder}.Result
   M=@n1.M
   Pre Skinned Local Position=@n1.M
   Distance Blend=@n1.M
   Fresnel=@n1.M
  ) {materialFunction=&MF_CharacterParent_2023_PostFX}
 ]
}
